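; Looping gameplay cue notify for the Tac Overshield recharging state, managing the charge-up sound.
; On application it takes the first spawned audio component and, if valid, saves it as ChargeUpSound.
; On removal it fades that saved sound out linearly over a short time, provided it is still valid.
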(event EventOnApplication(EitherVFX-preferthisone)
  (params (MyTarget :type Actor%20Object%20Reference) (Parameters :type Gameplay%20Cue%20Parameters%20Structure%20%28by%20ref%29) (ParticleComponents :type Array%20of%20FXSystem%20Component%20Object%20References) (AudioComponents :type Array%20of%20Audio%20Component%20Object%20References) (BurstCameraShakeInstance :type Camera%20Shake%20Base%20Object%20Reference) (BurstDecalInstance :type Decal%20Actor%20Object%20Reference))
  (bind _output (Utilities|Array|Get(acopy) AudioComponents))
  ;[1 1 1] Application - Save out the begin effect so that it can be faded out on stop.
  (if (Utilities|IsValid _output :!class /Script/Engine.KismetSystemLibrary)
    (Variables|Default|SetChargeUpSound _output)))

(event EventOnRemoval(EitherVFX-preferthisversion)
  (params (MyTarget :type Actor%20Object%20Reference) (Parameters :type Gameplay%20Cue%20Parameters%20Structure%20%28by%20ref%29) (ParticleComponents :type Array%20of%20FXSystem%20Component%20Object%20References) (AudioComponents :type Array%20of%20Audio%20Component%20Object%20References) (BurstCameraShakeInstance :type Camera%20Shake%20Base%20Object%20Reference) (BurstDecalInstance :type Decal%20Actor%20Object%20Reference))
  ;[1 1 1] Removal - Fade out on stop
  (bind _chargeupsound (Variables|Default|GetChargeUpSound)
    (:then
      (Audio|Components|Audio|FadeOut _chargeupsound 0.75 0.0 "Linear" :!class /Script/Engine.AudioComponent))
    (:else)))

(fn ConstructionScript)
