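; Burst gameplay cue notify for the TacOvershield breaking, handling its sound and a shield-break event.
; On burst it picks a first-person, third-person 2D or third-person 3D break cue from the effect causer via GetAudioforDamageEvent and plays it at the target's location.
; It also casts the effect causer to a player state, gets its Athena pawn, and fires that pawn's OnShieldBreak dispatcher only when the pawn is locally viewed.
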(fn OnBurst(EitherVFX-preferthisversion)
  (params (MyTarget :type Actor%20Object%20Reference) (Parameters :type Gameplay%20Cue%20Parameters%20Structure%20%28by%20ref%29) (ParticleComponents :type Array%20of%20FXSystem%20Component%20Object%20References) (AudioComponents :type Array%20of%20Audio%20Component%20Object%20References) (BurstCameraShakeInstance :type Camera%20Shake%20Base%20Object%20Reference) (BurstDecalInstance :type Decal%20Actor%20Object%20Reference))
  (bind (_normalizedmagnitude _rawmagnitude _effectcontext _matchedtagname _originaltag _aggregatedsourcetags _aggregatedtargettags _location _normal _instigator _effectcauser) (Ability|GameplayCue|BreakGameplayCueParameters (Variables|GetParameters)))
  (bind (_normalizedmagnitude_2 _rawmagnitude_2 _effectcontext_2 _matchedtagname_2 _originaltag_2 _aggregatedsourcetags_2 _aggregatedtargettags_2 _location_2 _normal_2 _instigator_2 _effectcauser_2) (Ability|GameplayCue|BreakGameplayCueParameters (Variables|GetParameters)))
  (Audio|PlaySoundatLocation 0 (AudioGameplay|GetAudioforDamageEvent 0 _effectcauser "/TacOvershield/Sounds/TacOvershield_Break_1P_Cue.TacOvershield_Break_1P_Cue" "/TacOvershield/Sounds/TacOvershield_Break_3P_2D_Cue.TacOvershield_Break_3P_2D_Cue" "/TacOvershield/Sounds/TacOvershield_Break_3P_3D_Cue.TacOvershield_Break_3P_3D_Cue") (Transformation|GetActorLocation (Variables|GetMyTarget)))
  (bind _asfort_player_state (Utilities|Casting|CastToFortPlayerState _effectcauser_2))
  (bind _asplayer_pawn_athena (Utilities|Casting|CastToPlayerPawn_Athena (PlayerState|GetPawn _asfort_player_state)))
  (if (Pawn|IsLocallyViewed _asplayer_pawn_athena)
    (EventDispatchers|CallOnShieldBreak _asplayer_pawn_athena true)
    (return)))
